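# Physics asset for the Rebirth Soldier's "Jacket No Back" variant, covering dangling details: chest and collar zippers, two front strap clips and an armband.
# Small capsule bodies on the dyn bones are constrained to MSTR anchor bodies.
# The armband knot branches into high and low two-bone chains with tighter swing limit values.
# Strap clips lock twist and one swing axis, so they swing like hinges.
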PhysicsAsset /Game/Characters/Player/Male/Medium/Bodies/M_MED_Rebirth_Soldier/Meshes/M_MED_Rebirth_Soldier_Dyn_Jacket_No_Back
bodies {
 dyn_armband_hi_01 [{shapeType=Capsule, length=2, radius=0.7162, center=<0.8 0 0>, rotation=<rot 90 0 0>}]
 dyn_armband_hi_02 [{shapeType=Capsule, length=2, radius=0.7162, center=<0.8 0 0>, rotation=<rot 90 0 0>}]
 dyn_armband_knot [{shapeType=Capsule, length=2, radius=0.7162, center=<0.05 0 0>, rotation=<rot 90 0 0>}]
 dyn_armband_knot_MSTR [{shapeType=Capsule, length=0.505, radius=0.505, rotation=<rot 0 -90 0>}]
 dyn_armband_low_01 [{shapeType=Capsule, length=2, radius=0.7162, center=<0.8 0 0>, rotation=<rot 90 0 0>}]
 dyn_armband_low_02 [{shapeType=Capsule, length=2, radius=0.7162, center=<0.8 0 0>, rotation=<rot 90 0 0>}]
 dyn_chest_zipper_r [{shapeType=Capsule, length=2, radius=0.7162, center=<1.9852 0 0.0002>, rotation=<rot 90 0 0>}]
 dyn_chest_zipper_r_MSTR [{shapeType=Capsule, length=0.5, radius=0.5, rotation=<rot 90 0 0>}]
 dyn_collar_zipper [{shapeType=Capsule, length=2, radius=0.7162, center=<1.9852 0 0.0002>, rotation=<rot 90 0 0>}]
 dyn_collar_zipper_MSTR [{shapeType=Capsule, length=0.5, radius=0.5, rotation=<rot 90 0 0>}]
 dyn_strap_clip_fr_l [{shapeType=Capsule, length=2, radius=0.7162, center=<1.1193 0.0001 0.0002>, rotation=<rot 90 0 0>}]
 dyn_strap_clip_fr_l_MSTR [{shapeType=Capsule, length=0.1, radius=0.1, rotation=<rot 90 0 0>}]
 dyn_strap_clip_fr_r [{shapeType=Capsule, length=2, radius=0.7162, center=<1.1193 0.0001 0.0002>, rotation=<rot 90 0 0>}]
 dyn_strap_clip_fr_r_MSTR [{shapeType=Capsule, length=0.1, radius=0.1, rotation=<rot 90 0 0>}]
}
constraints [
 {bone1Name=dyn_chest_zipper_r, bone2Name=dyn_chest_zipper_r_MSTR, swing1LimitDegrees=45, swing2LimitDegrees=45, twistLimitDegrees=2}
 {bone1Name=dyn_collar_zipper, bone2Name=dyn_collar_zipper_MSTR, swing1LimitDegrees=45, swing2LimitDegrees=45, twistLimitDegrees=2}
 {
  bone1Name=dyn_strap_clip_fr_l
  bone2Name=dyn_strap_clip_fr_l_MSTR
  swing1LimitDegrees=45
  swing1Motion=Locked
  swing2LimitDegrees=45
  twistLimitDegrees=45
  twistMotion=Locked
 }
 {bone1Name=dyn_armband_knot, bone2Name=dyn_armband_knot_MSTR, swing1LimitDegrees=45, swing2LimitDegrees=45, twistLimitDegrees=2}
 {bone1Name=dyn_armband_low_01, bone2Name=dyn_armband_knot, swing1LimitDegrees=35, swing2LimitDegrees=35, twistLimitDegrees=2}
 {bone1Name=dyn_armband_low_02, bone2Name=dyn_armband_low_01, swing1LimitDegrees=35, swing2LimitDegrees=35, twistLimitDegrees=2}
 {bone1Name=dyn_armband_hi_01, bone2Name=dyn_armband_knot, swing1LimitDegrees=35, swing2LimitDegrees=35, twistLimitDegrees=2}
 {bone1Name=dyn_armband_hi_02, bone2Name=dyn_armband_hi_01, swing1LimitDegrees=35, swing2LimitDegrees=35, twistLimitDegrees=2}
 {
  bone1Name=dyn_strap_clip_fr_r
  bone2Name=dyn_strap_clip_fr_r_MSTR
  swing1LimitDegrees=45
  swing1Motion=Locked
  swing2LimitDegrees=45
  twistLimitDegrees=45
  twistMotion=Locked
 }
]
