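# Physics asset for the dynamic leg parts of the male medium Rebirth Soldier player body, covering four boot zipper bones: inner and outer on each boot.
# Each zipper has an offset capsule body paired with a small capsule on its MSTR bone.
# One constraint per pair limits swing to a moderate cone and twist to a very narrow range, so each zipper can dangle with little spin.
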PhysicsAsset /Game/Characters/Player/Male/Medium/Bodies/M_MED_Rebirth_Soldier/Meshes/M_MED_Rebirth_Soldier_Dyn_Legs
bodies {
 dyn_boot_zipper_in_l [{shapeType=Capsule, length=2, radius=0.7162, center=<1.9852 0 0.0002>, rotation=<rot 90 0 0>}]
 dyn_boot_zipper_in_l_MSTR [{shapeType=Capsule, length=0.5, radius=0.5, rotation=<rot 90 0 0>}]
 dyn_boot_zipper_in_r [{shapeType=Capsule, length=2, radius=0.7162, center=<1.9852 0 0.0002>, rotation=<rot 90 0 0>}]
 dyn_boot_zipper_in_r_MSTR [{shapeType=Capsule, length=0.5, radius=0.5, rotation=<rot 90 0 0>}]
 dyn_boot_zipper_out_l [{shapeType=Capsule, length=2, radius=0.7162, center=<1.9852 0 0.0002>, rotation=<rot 90 0 0>}]
 dyn_boot_zipper_out_l_MSTR [{shapeType=Capsule, length=0.5, radius=0.5, rotation=<rot 90 0 0>}]
 dyn_boot_zipper_out_r [{shapeType=Capsule, length=2, radius=0.7162, center=<1.9852 0 0.0002>, rotation=<rot 90 0 0>}]
 dyn_boot_zipper_out_r_MSTR [{shapeType=Capsule, length=0.5, radius=0.5, rotation=<rot 90 0 0>}]
}
constraints [
 {bone1Name=dyn_boot_zipper_out_l, bone2Name=dyn_boot_zipper_out_l_MSTR, swing1LimitDegrees=45, swing2LimitDegrees=45, twistLimitDegrees=2}
 {bone1Name=dyn_boot_zipper_in_l, bone2Name=dyn_boot_zipper_in_l_MSTR, swing1LimitDegrees=45, swing2LimitDegrees=45, twistLimitDegrees=2}
 {bone1Name=dyn_boot_zipper_in_r, bone2Name=dyn_boot_zipper_in_r_MSTR, swing1LimitDegrees=45, swing2LimitDegrees=45, twistLimitDegrees=2}
 {bone1Name=dyn_boot_zipper_out_r, bone2Name=dyn_boot_zipper_out_r_MSTR, swing1LimitDegrees=45, swing2LimitDegrees=45, twistLimitDegrees=2}
]
