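# Secondary-motion AnimBP for the M_MED_Neon_Animal player body, layered over a pose copied from the attached parent mesh.
# Alpha-driven bone offsets adjust pelvis and kneepads, while weighted constraints place straps, belt items, pouches, kneepads and backpack.
# A RigidBody node simulates legs before the pose is cached.
# Two RigidBody jacket simulations branch from that cache, bIsWearingBackbling selecting the no-back variant, all reusing Rebirth Soldier physics assets.
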AnimBlueprint /Game/Characters/Player/Male/Medium/Bodies/M_MED_Neon_Animal/Meshes/M_MED_Neon_Animal_AnimBP_PreNativize
target_skeleton=&/M_MED_Neon_Animal_Skeleton
graphs [
 AnimGraph {nodes [
   n1 = RigidBody(ComponentPose=Local To Component(LocalPose=Use cached pose 'preJacketCache'())) {
    bClampLinearTranslationLimitToRefPose=true
    bOverrideWorldGravity=true
    overridePhysicsAsset=/Game/Characters/Player/Male/Medium/Bodies/M_MED_Rebirth_Soldier/Meshes/M_MED_Rebirth_Soldier_Dyn_Jacket
    simulationSpace=BaseBoneSpace
    alphaBoolBlend {blendOption=Linear}
    baseBoneRef {boneName=spine_01}
    componentAppliedLinearAccClamp=<3000 3000 3000>
    componentLinearAccScale=<0.5 0.2 0.2>
    componentLinearVelScale=<0.5 0.2 0.2>
    overrideWorldGravity {z=-980}
   }
   n2 = RigidBody(ComponentPose=Local To Component(LocalPose=Use cached pose 'preJacketCache'())) {
    bClampLinearTranslationLimitToRefPose=true
    bOverrideWorldGravity=true
    overridePhysicsAsset=/Game/Characters/Player/Male/Medium/Bodies/M_MED_Rebirth_Soldier/Meshes/M_MED_Rebirth_Soldier_Dyn_Jacket_No_Back
    simulationSpace=BaseBoneSpace
    alphaBoolBlend {blendOption=Linear}
    baseBoneRef {boneName=spine_01}
    componentAppliedLinearAccClamp=<3000 3000 3000>
    componentLinearAccScale=<0.5 0.2 0.2>
    componentLinearVelScale=<0.5 0.2 0.2>
    overrideWorldGravity {z=-980}
   }
   Output Pose(Result=Blend Poses by bool(
     BlendPose_0=Component To Local(ComponentPose=@n2)
     BlendPose_1=Component To Local(ComponentPose=@n1)
     bActiveValue=Get bIsWearingBackbling() -> (bIsWearingBackbling:bool)
    ) {blendType=Linear}) {layerGroup=None}
   n3 = Get ThighLeftRotAlpha() -> (ThighLeftRotAlpha:float)
   n4 = "Transform (Modify) Bone - Bone: pelvis_l"(
    ComponentPose=Local To Component(LocalPose=Copy Pose From Mesh() {bUseAttachedParent=true})
    Alpha=@n3
    Translation="-2.151,-0.099,0.216"
    Rotation="0,0,0"
    Scale="1.000,1.207,1.000"
   ) {
    scaleMode=BMM_Additive
    translationMode=BMM_Additive
    translationSpace=BCS_ParentBoneSpace
    alphaBoolBlend {blendOption=Linear}
    boneToModify {boneName=pelvis_l}
    scale=<1.0001 1.207>
    translation=<-2.151 -0.099 0.216>
   }
   n5 = "Transform (Modify) Bone - Bone: pelvis_mid_l"(ComponentPose=@n4, Alpha=@n3, Translation="-0.891,0.056,0.098", Rotation="0,0,0", Scale="1.523,1.000,1.000") {
    scaleMode=BMM_Additive
    translationMode=BMM_Additive
    translationSpace=BCS_ParentBoneSpace
    alphaBoolBlend {blendOption=Linear}
    boneToModify {boneName=pelvis_mid_l}
    scale {x=1.5231}
    translation=<-0.891 0.056 0.098>
   }
   n6 = "Transform (Modify) Bone - Bone: pelvis_side_l"(ComponentPose=@n5, Alpha=@n3, Translation="0,0,0", Rotation="0,0,0", Scale="1.110,1.000,1.038") {
    scaleMode=BMM_Additive
    translationMode=BMM_Additive
    translationSpace=BCS_ParentBoneSpace
    alphaBoolBlend {blendOption=Linear}
    boneToModify {boneName=pelvis_side_l}
    scale {x=1.11, z=1.038}
   }
   n7 = Get ThighRightRotAlpha() -> (ThighRightRotAlpha:float)
   n8 = "Transform (Modify) Bone - Bone: pelvis_r"(ComponentPose=@n6, Alpha=@n7, Translation="0,0,0", Rotation="0,0,0", Scale="1,1,1") {
    scaleMode=BMM_Additive
    translationMode=BMM_Additive
    translationSpace=BCS_ParentBoneSpace
    alphaBoolBlend {blendOption=Linear}
    boneToModify {boneName=pelvis_r}
   }
   n9 = "Transform (Modify) Bone - Bone: pelvis_mid_r"(ComponentPose=@n8, Alpha=@n7, Translation="0,0,0", Rotation="0,0,0", Scale="1.5,1,1.000") {
    scaleMode=BMM_Additive
    translationMode=BMM_Additive
    translationSpace=BCS_ParentBoneSpace
    alphaBoolBlend {blendOption=Linear}
    boneToModify {boneName=pelvis_mid_r}
    scale {x=1.5}
   }
   n10 = "Transform (Modify) Bone - Bone: pelvis_side_r"(ComponentPose=@n9, Alpha=@n7, Translation="0,0,0", Rotation="0,0,0", Scale="1,1,1") {
    scaleMode=BMM_Additive
    translationMode=BMM_Additive
    translationSpace=BCS_ParentBoneSpace
    alphaBoolBlend {blendOption=Linear}
    boneToModify {boneName=pelvis_side_r}
   }
   n11 = "Constraint - dyn_strap_clip_fr_l_MSTR (PPPPPP)"!(
    ComponentPose=@n10
    Alpha=1
    ConstraintWeights_0=0.4021
    ConstraintWeights_1:float=0.0663
    ConstraintWeights_2:float=0.1299
    ConstraintWeights_3:float=0.174
    ConstraintWeights_4:float=0.0925
    ConstraintWeights_5:float=0.1352
   ) -> (Pose) {
    alphaBoolBlend {blendOption=Linear}
    boneToModify {boneName=dyn_strap_clip_fr_l_MSTR}
    constraintSetup [
     {targetBone {boneName=clavicle_l}}
     {targetBone {boneName=spine_02}}
     {targetBone {boneName=spine_03}}
     {targetBone {boneName=spine_04}}
     {targetBone {boneName=spine_05}}
     {targetBone {boneName=upperarm_twist_01_l}}
    ]
    constraintWeights [1,1,1,1,1,1]
   }
   n12 = "Constraint - dyn_strap_clip_fr_r_MSTR (PPPPPP)"!(
    ComponentPose=@n11
    Alpha=1
    ConstraintWeights_0=0.1344
    ConstraintWeights_1:float=0.0671
    ConstraintWeights_2:float=0.1311
    ConstraintWeights_3:float=0.1756
    ConstraintWeights_4:float=0.0909
    ConstraintWeights_5:float=0.4009
   ) -> (Pose) {
    alphaBoolBlend {blendOption=Linear}
    boneToModify {boneName=dyn_strap_clip_fr_r_MSTR}
    constraintSetup [
     {targetBone {boneName=upperarm_twist_01_r}}
     {targetBone {boneName=spine_02}}
     {targetBone {boneName=spine_03}}
     {targetBone {boneName=spine_04}}
     {targetBone {boneName=spine_05}}
     {targetBone {boneName=clavicle_r}}
    ]
    constraintWeights [1,1,1,1,1,1]
   }
   n13 = "Constraint - dyn_strap_clip_bk_l_MSTR (PPPPP)"!(
    ComponentPose=@n12
    Alpha=1
    ConstraintWeights_0=0.1812
    ConstraintWeights_1:float=0.0801
    ConstraintWeights_2:float=0.0469
    ConstraintWeights_3:float=0.5755
    ConstraintWeights_4:float=0.1163
   ) -> (Pose) {
    alphaBoolBlend {blendOption=Linear}
    boneToModify {boneName=dyn_strap_clip_bk_l_MSTR}
    constraintSetup [
     {targetBone {boneName=spine_03}}
     {targetBone {boneName=spine_04}}
     {targetBone {boneName=spine_05}}
     {targetBone {boneName=clavicle_l}}
     {targetBone {boneName=upperarm_twist_01_l}}
    ]
    constraintWeights [1,1,1,1,1]
   }
   n14 = "Constraint - dyn_strap_clip_bk_r_MSTR (PPPPP)"!(
    ComponentPose=@n13
    Alpha=1
    ConstraintWeights_0=0.1158
    ConstraintWeights_1:float=0.5736
    ConstraintWeights_2:float=0.1832
    ConstraintWeights_3:float=0.0828
    ConstraintWeights_4:float=0.0446
   ) -> (Pose) {
    alphaBoolBlend {blendOption=Linear}
    boneToModify {boneName=dyn_strap_clip_bk_r_MSTR}
    constraintSetup [
     {targetBone {boneName=upperarm_twist_01_r}}
     {targetBone {boneName=clavicle_r}}
     {targetBone {boneName=spine_03}}
     {targetBone {boneName=spine_04}}
     {targetBone {boneName=spine_05}}
    ]
    constraintWeights [1,1,1,1,1]
   }
   n15 = "Constraint - dyn_belt_bullet_01 (PPPPPP)"!(
    ComponentPose=@n14
    Alpha=1
    ConstraintWeights_0=0.3893
    ConstraintWeights_1:float=0.1443
    ConstraintWeights_2:float=0.0624
    ConstraintWeights_3:float=0.0001
    ConstraintWeights_4:float=0
    ConstraintWeights_5:float=0.4039
   ) -> (Pose) {
    alphaBoolBlend {blendOption=Linear}
    boneToModify {boneName=dyn_belt_bullet_01}
    constraintSetup [
     {targetBone {boneName=spine_01}}
     {targetBone {boneName=spine_02}}
     {targetBone {boneName=spine_03}}
     {targetBone {boneName=spine_04}}
     {targetBone {boneName=clavicle_r}}
     {targetBone {boneName=pelvis}}
    ]
    constraintWeights [1,1,1,1,1,1]
   }
   n16 = "Constraint - dyn_belt_bullet_02 (PPPPPP)"!(
    ComponentPose=@n15
    Alpha=1
    ConstraintWeights_0=0.388
    ConstraintWeights_1:float=0.1398
    ConstraintWeights_2:float=0.064
    ConstraintWeights_3:float=0.0002
    ConstraintWeights_4:float=0.0001
    ConstraintWeights_5:float=0.408
   ) -> (Pose) {
    alphaBoolBlend {blendOption=Linear}
    boneToModify {boneName=dyn_belt_bullet_02}
    constraintSetup [
     {targetBone {boneName=spine_01}}
     {targetBone {boneName=spine_02}}
     {targetBone {boneName=spine_03}}
     {targetBone {boneName=spine_04}}
     {targetBone {boneName=clavicle_r}}
     {targetBone {boneName=pelvis}}
    ]
    constraintWeights [1,1,1,1,1,1]
   }
   n17 = "Constraint - dyn_belt_bullet_03 (PPPPP)"!(
    ComponentPose=@n16
    Alpha=1
    ConstraintWeights_0=0.3741
    ConstraintWeights_1:float=0.1365
    ConstraintWeights_2:float=0.0683
    ConstraintWeights_3:float=0.0005
    ConstraintWeights_4:float=0.4206
   ) -> (Pose) {
    alphaBoolBlend {blendOption=Linear}
    boneToModify {boneName=dyn_belt_bullet_03}
    constraintSetup [
     {targetBone {boneName=spine_01}}
     {targetBone {boneName=spine_02}}
     {targetBone {boneName=spine_03}}
     {targetBone {boneName=spine_04}}
     {targetBone {boneName=pelvis}}
    ]
    constraintWeights [1,1,1,1,1]
   }
   n18 = "Constraint - dyn_belt_back_pouch (PPPP)"!(
    ComponentPose=@n17
    Alpha=1
    ConstraintWeights_0=0.1735
    ConstraintWeights_1:float=0.0357
    ConstraintWeights_2:float=0.0459
    ConstraintWeights_3:float=0.7448
   ) -> (Pose) {
    alphaBoolBlend {blendOption=Linear}
    boneToModify {boneName=dyn_belt_back_pouch}
    constraintSetup [{targetBone {boneName=spine_01}}, {targetBone {boneName=spine_02}}, {targetBone {boneName=spine_03}}, {targetBone {boneName=pelvis}}]
    constraintWeights [1,1,1,1]
   }
   n19 = "Constraint - dyn_hip_pouch_bk_r (PPP)"!(ComponentPose=@n18, Alpha=1, ConstraintWeights_0=0.0155, ConstraintWeights_1:float=0.3262, ConstraintWeights_2:float=0.6583) -> (Pose) {
    alphaBoolBlend {blendOption=Linear}
    boneToModify {boneName=dyn_hip_pouch_bk_r}
    constraintSetup [{targetBone {boneName=thigh_r}}, {targetBone {boneName=pelvis}}, {targetBone {boneName=thigh_twist_01_r}}]
    constraintWeights [1,1,1]
   }
   n20 = "Constraint - dyn_hip_pouch_fr_r (PP)"!(ComponentPose=@n19, Alpha=1, ConstraintWeights_0=0.2777, ConstraintWeights_1:float=0.7223) -> (Pose) {
    alphaBoolBlend {blendOption=Linear}
    boneToModify {boneName=dyn_hip_pouch_fr_r}
    constraintSetup [{targetBone {boneName=pelvis}}, {targetBone {boneName=thigh_twist_01_r}}]
    constraintWeights [1,1]
   }
   n21 = "Constraint - dyn_hip_pouch_fr_l (PPP)"!(ComponentPose=@n20, Alpha=1, ConstraintWeights_0=0.0775, ConstraintWeights_1:float=0.8334, ConstraintWeights_2:float=0.089) -> (Pose) {
    alphaBoolBlend {blendOption=Linear}
    boneToModify {boneName=dyn_hip_pouch_fr_l}
    constraintSetup [{targetBone {boneName=pelvis}}, {targetBone {boneName=thigh_twist_01_l}}, {targetBone {boneName=thigh_l}}]
    constraintWeights [1,1,1]
   }
   n22 = "Constraint - dyn_kneepad_l (PP)"!(ComponentPose=@n21, Alpha=1, ConstraintWeights_0=0.7, ConstraintWeights_1:float=0.3) -> (Pose) {
    alphaBoolBlend {blendOption=Linear}
    boneToModify {boneName=dyn_kneepad_l}
    constraintSetup [{targetBone {boneName=calf_l}}, {targetBone {boneName=thigh_l}}]
    constraintWeights [1,1]
   }
   n23 = "Constraint - dyn_kneepad_r (PP)"!(ComponentPose=@n22, Alpha=1, ConstraintWeights_0=0.3, ConstraintWeights_1:float=0.7) -> (Pose) {
    alphaBoolBlend {blendOption=Linear}
    boneToModify {boneName=dyn_kneepad_r}
    constraintSetup [{targetBone {boneName=calf_r}}, {targetBone {boneName=thigh_r}}]
    constraintWeights [1,1]
   }
   n24 = "Transform (Modify) Bone - Bone: dyn_kneepad_l"(ComponentPose=@n23, Alpha=Get CalfLeftRotAlpha() -> (CalfLeftRotAlpha:float), Translation="0.000,-3,0.000", Rotation="0,0,0", Scale="1,1,1") {
    rotationMode=BMM_Additive
    rotationSpace=BCS_BoneSpace
    translationMode=BMM_Additive
    translationSpace=BCS_BoneSpace
    alphaBoolBlend {blendOption=Linear}
    alphaScaleBiasClamp {interpSpeedDecreasing=1, interpSpeedIncreasing=1}
    boneToModify {boneName=dyn_kneepad_l}
    translation {y=-3}
   }
   n25 = "Transform (Modify) Bone - Bone: dyn_kneepad_r"(ComponentPose=@n24, Alpha=Get CalfRightRotAlpha() -> (CalfRightRotAlpha:float), Translation="0.000,3,0.000", Rotation="0,0,0", Scale="1,1,1") {
    rotationMode=BMM_Additive
    rotationSpace=BCS_BoneSpace
    translationMode=BMM_Additive
    translationSpace=BCS_BoneSpace
    alphaBoolBlend {blendOption=Linear}
    alphaScaleBiasClamp {interpSpeedDecreasing=1, interpSpeedIncreasing=1}
    boneToModify {boneName=dyn_kneepad_r}
    translation {y=3}
   }
   n26 = RigidBody(ComponentPose=@n25) {
    bClampLinearTranslationLimitToRefPose=true
    bOverrideWorldGravity=true
    overridePhysicsAsset=/Game/Characters/Player/Male/Medium/Bodies/M_MED_Rebirth_Soldier/Meshes/M_MED_Rebirth_Soldier_Dyn_Legs
    simulationSpace=BaseBoneSpace
    alphaBoolBlend {blendOption=Linear}
    baseBoneRef {boneName=thigh_l}
    componentAppliedLinearAccClamp=<3000 3000 3000>
    componentLinearAccScale=<0.12 0.4 0.4>
    componentLinearVelScale=<0.12 0.4 0.4>
    overrideWorldGravity {z=-980}
   }
   n27 = "Constraint - dyn_backpack (PPPPP)"!(
    ComponentPose=@n26
    Alpha=1
    ConstraintWeights_0=0.2133
    ConstraintWeights_1:float=0.3414
    ConstraintWeights_2:float=0.1453
    ConstraintWeights_3:float=0.0432
    ConstraintWeights_4:float=0.2567
   ) -> (Pose) {
    alphaBoolBlend {blendOption=Linear}
    boneToModify {boneName=dyn_backpack}
    constraintSetup [
     {targetBone {boneName=clavicle_r}}
     {targetBone {boneName=spine_03}}
     {targetBone {boneName=spine_04}}
     {targetBone {boneName=spine_05}}
     {targetBone {boneName=clavicle_l}}
    ]
    constraintWeights [1,1,1,1,1]
   }
   Save cached pose 'preJacketCache'(Pose=Component To Local(ComponentPose=@n27)) {cacheName=preJacketCache}
  ]}
 EventGraph {links [], nodes []}
]
variables []
